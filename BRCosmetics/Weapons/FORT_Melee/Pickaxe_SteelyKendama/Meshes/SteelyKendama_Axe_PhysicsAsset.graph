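# Physics asset for the Steely Kendama pickaxe, a cosmetic two-handed melee weapon mesh.
# It holds a single capsule body on the Melee_TwoHanded_Handle bone, offset and rotated from that bone.
# There are no constraints, so the asset defines one collision shape with no jointed or articulated parts.
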PhysicsAsset /BRCosmetics/Weapons/FORT_Melee/Pickaxe_SteelyKendama/Meshes/SteelyKendama_Axe_PhysicsAsset
bodies {Melee_TwoHanded_Handle [{shapeType=Capsule, length=28.5299, radius=7.4558, center=<0.5342 0.1671 15.4175>, rotation=<rot 1.0717 -90 3.6254>}]}
constraints []
